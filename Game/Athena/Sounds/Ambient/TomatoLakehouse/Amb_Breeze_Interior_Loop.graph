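# Ambient sound cue for a looping interior breeze bed at the Athena TomatoLakehouse location.
# A looping wave player feeds an Enveloper whose volume curve fades the breeze up from near silence over about five seconds.
# The cue plays through the indoor ambient sound class at reduced volume.
# Overridden attenuation uses a box shape with a very short falloff and no reverb send, confining the sound tightly to that volume.
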SoundCue /Game/Athena/Sounds/Ambient/TomatoLakehouse/Amb_Breeze_Interior_Loop
root=Enveloper_0
nodes [
 {
  id=Enveloper_0
  children [WavePlayer_0]
  properties {pitchCurve {editorCurveData {keys [{time=0}]}}, volumeCurve {editorCurveData {keys [<at 0.0512 0.0468>,<at 5.2894 1>]}}}
 }
 {id=WavePlayer_0, properties {bLooping=true, soundWaveAssetPtr=&/Amb_Breeze_Lakehouse_Loop_01}}
]
properties {
 bOverrideAttenuation=true
 maxDistance=785.4984
 soundClassObject=&Fort_Ambient_Indoors_Athena
 volumeMultiplier=0.451
 attenuationOverrides {
  attenuationShape=Box
  bEnableReverbSend=false
  falloffDistance=15.6325
  stereoSpread=218581.8125
  attenuationShapeExtents=<769.8658 766.2503 640.8283>
 }
}
